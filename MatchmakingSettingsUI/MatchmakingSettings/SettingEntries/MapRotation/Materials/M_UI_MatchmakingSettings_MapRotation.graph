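# Translucent UI material for the map rotation entry in matchmaking settings, showing a texture over a CoreUI backplate.
# The Texture parameter is scaled to fill from a top-center pivot, using the texture's own size.
# A smoothstep on normalized horizontal UV blends from pure backplate on the left to a partially visible image on the right.
# A CoreUI rounded mask function supplies rounded corners.
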Material /MatchmakingSettingsUI/MatchmakingSettings/SettingEntries/MapRotation/Materials/M_UI_MatchmakingSettings_MapRotation
graph {
 properties {bAutomaticallySetUsageInEditor=false, blendMode=BLEND_Translucent, materialDomain=MD_UI, usedShadingModels=Default Lit}
 nodes [
  n1 = TextureObjectParameter() {parameterName=Texture}
  n2 = TextureProperty(Tex=@n1)
  n3 = Constant2Vector() {r=0.5}
  n4 = ComponentMask(None=MaterialFunctionCall() {materialFunction=&GetUserInterfaceUV}.Normalized UV) {r=true}
  LinearInterpolate(
   A=MaterialFunctionCall!(Post Buffer Blur) {materialFunction=/FortUILibrary/MaterialFunctions/CoreUI/MF_UI_CoreUI_BackplateC}.Result
   B=TextureSample(
    UVs=MaterialFunctionCall(TextureSize=@n2, ScalePivotPoint=@n3) {materialFunction=&MF_UI_UVScaling}.Scaled To Fill
    Tex=@n1
   ).RGB
   Alpha=LinearInterpolate(Alpha=SmoothStep(Value=@n4)) {constB=0.6}
  )
  MaterialFunctionCall(Corner Radius=Constant() {r=10}, Radius in Pixels=Constant() {r=1}) {materialFunction=&MF_UI_CoreUI_RoundedMask}
 ]
}
