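# Layout-only UserWidget that displays a map entry for the map rotation in matchmaking settings.
# An Overlay uses a Spacer for minimum size, beneath a filling map image that draws the MapRotation UI material instance.
# A padded, vertically centered box stacks the wrapping map name text above a named slot for extra map info.
# The event graph is empty; the image, text and slot are exposed as variables.
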WidgetBlueprint /MatchmakingSettingsUI/MatchmakingSettings/SettingEntries/MapRotation/WBP_MatchmakingSettings_MapRotationWidget
graphs [EventGraph {links [], nodes []}]
variables []
widgets {
 info {inheritedWidgetCount=0, parentClass=&UserWidget, rootWidgetClass=&Overlay, widgetCount=6}
 widgets [
  Overlay_19 {widgetClassPath=&Overlay, widget=Overlay_19}
  MinSize {
   widgetClassPath=&Spacer
   parent=Overlay_19
   widget=MinSize
   properties {size=<448 120>}
   slot {class=OverlaySlot, properties {horizontalAlignment=HAlign_Fill, verticalAlignment=VAlign_Fill}}
  }
  Img_Map {
   widgetClassPath=&Image
   bIsVariable=true
   parent=Overlay_19
   widget=Img_Map
   properties {
    visibility=HitTestInvisible
    brush {
     bIsDynamicallyLoaded=false
     drawAs=Image
     imageType=FullColor
     mirroring=NoMirror
     resourceName=None
     resourceObject=/MatchmakingSettingsUI/MatchmakingSettings/SettingEntries/MapRotation/MaterialInstances/MI_UI_MatchmakingSettings_MapRotation
     tiling=NoTile
     imageSize=<32 32>
     margin=<ltrb 0 0 0 0>
     outlineSettings {
      bUseBrushTransparency=false
      roundingType=HalfHeightRadius
      width=0
      color {colorUseRule=UseColor_Specified, specifiedColor=<rgba 0 0 0 0>}
      cornerRadii=<0 0 0 0>
     }
     tintColor {colorUseRule=UseColor_Specified, specifiedColor=<rgba 1 1 1 1>}
     uVRegion {bIsValid=false, max=<0 0>, min=<0 0>}
    }
   }
   slot {class=OverlaySlot, properties {horizontalAlignment=HAlign_Fill, verticalAlignment=VAlign_Fill}}
  }
  VerticalBox_31 {
   widgetClassPath=&VerticalBox
   parent=Overlay_19
   widget=VerticalBox_31
   slot {class=OverlaySlot, properties {horizontalAlignment=HAlign_Fill, verticalAlignment=VAlign_Center, padding=<ltrb 32 16 32 16>}}
  }
  Txt_MapName {
   widgetClassPath=&CommonTextBlock
   bIsVariable=true
   parent=VerticalBox_31
   widget=Txt_MapName
   properties {
    autoWrapText=true
    scrollStyle=&TextScrollStyle-Base_C
    text=$MapName
    font {
     bForceMonospaced=false
     bMaterialIsStencil=false
     fontMaterial=null
     fontObject=&NotoSans
     letterSpacing=0
     monospacedWidth=1
     size=24
     skewAmount=0
     typefaceFontName=NotoSans-Regular
     outlineSettings {
      bApplyOutlineToDropShadows=false
      bMiteredCorners=false
      bSeparateFillAlpha=false
      outlineMaterial=null
      outlineSize=0
      outlineColor=<rgba 0 0 0 1>
     }
    }
    shadowOffset=<0 0>
   }
   slot {class=VerticalBoxSlot}
  }
  NamedSlot_MapInfo {
   widgetClassPath=&NamedSlot
   bIsVariable=true
   parent=VerticalBox_31
   widget=NamedSlot_MapInfo
   slot {class=VerticalBoxSlot, properties {padding=<ltrb 0 6 0 0>}}
  }
 ]
}
